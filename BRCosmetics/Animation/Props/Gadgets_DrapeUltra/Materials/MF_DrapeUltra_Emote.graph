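# Material function for the DrapeUltra emote that overrides a prop's opacity mask and emissive color, apparently revealing swords through portals.
# World position is transformed into a slightly rotated portal-relative frame, then its height is stepped for masking and noise-perturbed for emissive glow.
# Sphere masks on time since GameTimeInput provide glow and full-opacity timing.
# A/B parameters blend per sword, with orange emissive suppressed when PortalBLocation is zero.
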MaterialFunction /BRCosmetics/Animation/Props/Gadgets_DrapeUltra/Materials/MF_DrapeUltra_Emote
graph {nodes [
  Add(A=CameraPositionWS(), B=WorldPosition() {worldPositionShaderOffset=WPT_CameraRelative}.XYZ)
  n1 = TextureCoordinate()
  n2 = ScalarParameter() {defaultValue=2, parameterName=AlphaMaskNoiseTiling}
  n3 = Constant()
  n4 = ScalarParameter() {defaultValue=-1, parameterName=AlphaMaskNoisePanningSpeed}
  n5 = TextureSampleParameter2D(UVs=Panner(Coordinate=Multiply(A=@n1, B=@n2), Speed=AppendVector(A=@n3, B=@n4))) {
   parameterName=NoiseMask
   samplerType=SAMPLERTYPE_LinearGrayscale
   texture=/BRCosmetics/Weapons/FORT_Melee/Pickaxe_Cube_Ninja_Male/FX/T_FluidFlameTile
  }
  ConstantBiasScale(None=@n5.R) {bias=-0.5, scale=2}
  FunctionOutput(None=SetMaterialAttributes!(MaterialAttributes=NamedRerouteUsage(), OpacityMask=NamedRerouteUsage(), EmissiveColor=NamedRerouteUsage()) {attributeSetTypes [#1, #2]})
  n6 = WorldPosition()
  n7 = NamedRerouteUsage()
  n8 = NamedRerouteUsage()
  n9 = ScalarParameter() {defaultValue=-0.02, parameterName=RotationAngle}
  n10 = Constant()
  n11 = NamedRerouteUsage()
  n12 = RotateAboutAxis(NormalizedRotationAxis=@n8, RotationAngle=@n9, PivotPoint=@n10, Position=@n11)
  n13 = NamedRerouteUsage()
  n14 = RotateAboutAxis(NormalizedRotationAxis=@n8, RotationAngle=@n9, PivotPoint=@n10, Position=@n13)
  n15 = ComponentMask(None=MaterialFunctionCall!(VectorToTransform=Subtract(A=@n6.XYZ, B=@n7), BasisX=@n8, BasisY=Add(A=@n12, B=@n11), BasisZ=Add(A=@n14, B=@n13)) {materialFunction=/Engine/Functions/Engine_MaterialFunctions02/Math/InverseTransformMatrix}.Transformed Vector 3) {b=true}
  n16 = Step(
   Y=LinearInterpolate(
    A=ScalarParameter() {defaultValue=6, parameterName=HeightOffsetA}
    B=ScalarParameter() {defaultValue=6, parameterName=HeightOffsetB}
    Alpha=NamedRerouteUsage()
   )
   X=@n15
  )
  n17 = NamedRerouteUsage()
  NamedRerouteDeclaration(None=LinearInterpolate(A=Max(A=@n16, B=@n17), Alpha=NamedRerouteUsage())) {name=OpacityMask, nodeColor=<rgba 0 0.3906 1 1>}
  NamedRerouteDeclaration(None=LinearInterpolate(
    A=VectorParameter() {parameterName=UpVectorA, primitiveDataIndex=4, defaultValue=<rgba 0 0 0 1>}.RGB
    B=VectorParameter() {parameterName=UpVectorA, primitiveDataIndex=4, defaultValue=<rgba 0 0 0 1>}.RGB
    Alpha=NamedRerouteUsage()
   )) {name=UpVector, nodeColor=<rgba 0.125 0 1 1>}
  n18 = Subtract(A=Time(), B=ScalarParameter() {parameterName=GameTimeInput})
  n19 = ScalarParameter() {defaultValue=0.5, parameterName=SwordGlowTimeInSeconds}
  n20 = ScalarParameter() {defaultValue=0.5, parameterName=SwordGlowTimeRange}
  n21 = ScalarParameter() {parameterName=Hardness}
  NamedRerouteDeclaration(None=Saturate(None=SphereMask(A=@n18, B=@n19, Radius=@n20, Hardness=@n21) {hardnessPercent=0})) {name=SwordGlowMask, nodeColor=<rgba 1 0 0.9844 1>}
  NamedRerouteDeclaration(None=ScalarParameter() {parameterName=MaterialCurveB}) {name=CurveInputB, nodeColor=<rgba 0.0078 1 0 1>}
  NamedRerouteDeclaration(None=ScalarParameter() {parameterName=MaterialCurveA}) {name=CurveInput, nodeColor=<rgba 0.0078 1 0 1>}
  NamedRerouteDeclaration(None=LinearInterpolate(Alpha=ScalarParameter() {parameterName=IsDrapeUltraLoud})) {name=SwordMask, nodeColor=<rgba 1 0 0.9141 1>}
  n22 = Sign(None=Length(None=VectorParameter() {parameterName=PortalBLocation, defaultValue=<rgba 0 0 0 1>}.RGB) {is=class MaterialExpressionLength})
  n23 = LinearInterpolate(
   A=VectorParameter() {parameterName=EmissiveColorA, defaultValue=<rgba 0 1 2 1>}.RGB
   B=LinearInterpolate(
    B=VectorParameter() {parameterName=EmissiveColorb, defaultValue=<rgba 5 0.5 0 1>}.RGB
    Alpha=Abs(None=@n22) {desc=Orange pickaxe sword doen'st have a portal}
   ) {desc=Pickaxe Orange Sword Doesn't have portal}
   Alpha=NamedRerouteUsage()
  )
  n24 = NamedRerouteUsage()
  n25 = ScalarParameter() {defaultValue=30, parameterName=AlphaMaskNoiseStrength}
  n26 = SphereMask(
   A=Add(A=@n15, B=Multiply(A=@n5.R, B=@n25))
   B=Constant()
   Radius=ScalarParameter() {defaultValue=30, parameterName=EmissiveRadius}
   Hardness=ScalarParameter() {defaultValue=0.5, parameterName=EmissiveHardness}
  )
  NamedRerouteDeclaration(None=LinearInterpolate(A=NamedRerouteUsage(), B=Add(A=@n23, B=@n24), Alpha=Saturate(None=@n26))) {name=EmissiveColor, nodeColor=<rgba 1 0 0.9375 1>}
  n27 = FunctionInput() {inputType=FunctionInput_MaterialAttributes}
  NamedRerouteDeclaration(None=GetMaterialAttributes(None=NamedRerouteDeclaration(None=@n27) {name=MA_IN, nodeColor=<rgba 1 0 0.0703 1>}) {attributeGetTypes [#2]}.EmissiveColor) {name=Emissive_IN, nodeColor=<rgba 1 0 0.9844 1>}
  NamedRerouteDeclaration(None=SphereMask(
    A=@n18
    B=ScalarParameter() {defaultValue=1, parameterName=FullOpaqueTimeInSeconds}
    Radius=ScalarParameter() {defaultValue=0.5, parameterName=FullOpaqueTimeRange}
    Hardness=ScalarParameter() {parameterName=Hardness}
   ) {hardnessPercent=0}) {name=SwordFullOpaqueMask, nodeColor=<rgba 0 0.3906 1 1>}
  NamedRerouteDeclaration(None=LinearInterpolate(
    A=VectorParameter() {parameterName=PortalALocation, defaultValue=<rgba 0 0 0 1>}.RGB
    B=VectorParameter() {parameterName=PortalBLocation, defaultValue=<rgba 0 0 0 1>}.RGB
    Alpha=NamedRerouteUsage()
   )) {name=PortalLocation, nodeColor=<rgba 0.2147 0.5099 1 1>}
  NamedRerouteDeclaration(None=LinearInterpolate(
    A=VectorParameter() {parameterName=ForwardVectorA, primitiveDataIndex=8, defaultValue=<rgba 0 0 0 1>}.RGB
    B=VectorParameter() {parameterName=ForwardVectorA, primitiveDataIndex=8, defaultValue=<rgba 0 0 0 1>}.RGB
    Alpha=NamedRerouteUsage()
   )) {name=ForwardVector, nodeColor=<rgba 0 1 0.7656 1>}
  NamedRerouteDeclaration(None=LinearInterpolate(
    A=VectorParameter() {parameterName=RightVectorA, primitiveDataIndex=12, defaultValue=<rgba 0 0 0 1>}.RGB
    B=VectorParameter() {parameterName=RightVectorA, primitiveDataIndex=12, defaultValue=<rgba 0 0 0 1>}.RGB
    Alpha=NamedRerouteUsage()
   )) {name=RightVector, nodeColor=<rgba 1 0.4453 0 1>}
 ]}
