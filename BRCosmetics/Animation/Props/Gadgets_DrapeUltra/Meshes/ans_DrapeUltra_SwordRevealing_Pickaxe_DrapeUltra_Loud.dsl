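; Anim notify state for the DrapeUltra pickaxe's sword reveal (Loud variant), driving a portal-based effect in the mesh's dynamic material.
; On begin it writes the current game time to GameTimeInput as the effect's start stamp.
; Each tick it scans the attach parent's direct children for PortalA and PortalB Niagara systems and writes their world location and axes to material parameters.
; The axes are remapped: Up feeds Forward, Forward feeds Right, Right feeds Up.
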(fn Received_NotifyBegin
  (params (MeshComp :type Skeletal%20Mesh%20Component%20Object%20Reference) (Animation :type Anim%20Sequence%20Base%20Object%20Reference) (TotalDuration :type Float%20%28single-precision%29) (EventReference :type Anim%20Notify%20Event%20Reference%20Structure%20%28by%20ref%29))
  (bind _returnvalue (Fortnite|GetDynamicMaterialInstance MeshComp))
  (Rendering|Material|SetScalarParameterValue _returnvalue "GameTimeInput" (Utilities|Time|GetGameTimeinSeconds MeshComp) :!class /Script/Engine.MaterialInstanceDynamic)
  (return false))

(fn Received_NotifyTick
  (params (MeshComp :type Skeletal%20Mesh%20Component%20Object%20Reference) (Animation :type Anim%20Sequence%20Base%20Object%20Reference) (FrameDeltaTime :type Float%20%28single-precision%29) (EventReference :type Anim%20Notify%20Event%20Reference%20Structure%20%28by%20ref%29))
  (bind _meshcomp (Variables|GetMeshComp))
  (bind _array_element (Utilities|Array|ForEachLoopwithBreak (Components|GetChildrenComponents (Transformation|GetAttachParent MeshComp) false))
    (:LoopBody
      (bind _asniagara_particle_system_component (Utilities|Casting|CastToNiagaraComponent _array_element))
      (if (== (Niagara|GetNiagaraSystemAsset _asniagara_particle_system_component) (Variables|Default|GetPortalA))
        (bind _returnvalue (Fortnite|GetDynamicMaterialInstance _meshcomp))
        (Rendering|Material|SetVectorParameterValue _returnvalue "PortalALocation" (Math|Conversions|ToLinearColor(Vector) (Transformation|GetWorldLocation _asniagara_particle_system_component)) :!class /Script/Engine.MaterialInstanceDynamic)
        (Rendering|Material|SetVectorParameterValue _returnvalue "ForwardVectorA" (Math|Conversions|ToLinearColor(Vector) (Transformation|GetUpVector _asniagara_particle_system_component)) :!class /Script/Engine.MaterialInstanceDynamic)
        (Rendering|Material|SetVectorParameterValue _returnvalue "RightVectorA" (Math|Conversions|ToLinearColor(Vector) (Transformation|GetForwardVector _asniagara_particle_system_component)) :!class /Script/Engine.MaterialInstanceDynamic)
        (Rendering|Material|SetVectorParameterValue _returnvalue "UpVectorA" (Math|Conversions|ToLinearColor(Vector) (Transformation|GetRightVector _asniagara_particle_system_component)) :!class /Script/Engine.MaterialInstanceDynamic)
        (elif (== (Niagara|GetNiagaraSystemAsset _asniagara_particle_system_component) (Variables|Default|GetPortalB))
          (bind _returnvalue_2 (Fortnite|GetDynamicMaterialInstance _meshcomp))
          (Rendering|Material|SetVectorParameterValue _returnvalue_2 "PortalBLocation" (Math|Conversions|ToLinearColor(Vector) (Transformation|GetWorldLocation _asniagara_particle_system_component)) :!class /Script/Engine.MaterialInstanceDynamic)
          (Rendering|Material|SetVectorParameterValue _returnvalue_2 "ForwardVectorB" (Math|Conversions|ToLinearColor(Vector) (Transformation|GetUpVector _asniagara_particle_system_component)) :!class /Script/Engine.MaterialInstanceDynamic)
          (Rendering|Material|SetVectorParameterValue _returnvalue_2 "RightVectorB" (Math|Conversions|ToLinearColor(Vector) (Transformation|GetForwardVector _asniagara_particle_system_component)) :!class /Script/Engine.MaterialInstanceDynamic)
          (Rendering|Material|SetVectorParameterValue _returnvalue_2 "UpVectorB" (Math|Conversions|ToLinearColor(Vector) (Transformation|GetRightVector _asniagara_particle_system_component)) :!class /Script/Engine.MaterialInstanceDynamic))))
    (:Completed
      (return false))))
